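# Conversation database for the TennisLeash NPC, a Santa Shaq character, providing its "Ask about..." dialogue branch.
# The entry tag leads to an opening Speech node that offers a "Naughty or Nice?" topic or a BACK choice.
# The topic plays two chained Speech lines of holiday banter, then a ReturnToStart node shown as a BACK choice.
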ConversationDatabase /NPCLibrary/NPCs/TennisLeash/NPCLibrary_Conversation_TennisLeash_Dialogue
entry_points [{tag=Conversation.Entry.TennisLeash.Dialogue, destinations [#1]}]
nodes [
 ConversationEntryPointNode_0 {
  id=#1
  class=ConversationEntryPointNode
  outputs [#2]
  properties {entryTag {tagName=Conversation.Entry.TennisLeash.Dialogue}}
 }
 Speech_C_0 {
  id=#2
  class=Speech_C
  outputs [#3, #4]
  properties {nodeName="", softTaskIcon=&T_UI_NPC_question, generalConfig {defaultMessage=You call. Santa Shaq answers.}}
  sub_nodes [ConversationChoiceNode_5 {class=ConversationChoiceNode, properties {defaultChoiceDisplayText=Ask about...}}]
 }
 Back_C_0 {
  id=#4
  class=Back_C
  sub_nodes [ConversationChoiceNode_7 {class=ConversationChoiceNode, properties {defaultChoiceDisplayText=BACK}}]
 }
 Speech_C_7 {
  id=#3
  class=Speech_C
  outputs [#5]
  properties {
   nodeName=""
   softTaskIcon=&T_UI_NPC_question
   generalConfig {defaultMessage="Come on now. In your heart, you know if you've been naughty or nice this year."}
  }
  sub_nodes [ConversationChoiceNode_15 {class=ConversationChoiceNode, properties {defaultChoiceDisplayText=Naughty or Nice?}}]
 }
 Speech_C_15 {
  id=#5
  class=Speech_C
  outputs [#6]
  properties {
   nodeName=""
   softTaskIcon=&T_UI_NPC_question
   generalConfig {defaultMessage="But if you feel like you gotta ask... the answer's probably nice. Naughty people don't wanna know!"}
  }
 }
 ReturnToStart_C_23 {
  id=#6
  class=ReturnToStart_C
  properties {softTaskIcon=&T_UI_NPC_Back}
  sub_nodes [ConversationChoiceNode_32 {class=ConversationChoiceNode, properties {defaultChoiceDisplayText=BACK}}]
 }
]
